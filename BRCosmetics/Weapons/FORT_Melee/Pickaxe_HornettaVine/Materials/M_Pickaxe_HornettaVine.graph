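# Master material for the HornettaVine pickaxe, built on the shared character function stack with animated-vein style emissive borrowed from the HornettaVine body.
# PreFX, CharacterSkinFX and PostFX functions build the material attributes, whose emissive is then multiplied by deep veins and a noisy glow gradient.
# A sharpened, boosted SkinFX green-channel mask and per-time-of-day scalars scale that glow before CINE_MF_CharPostFX finishes the attributes.
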Material /BRCosmetics/Weapons/FORT_Melee/Pickaxe_HornettaVine/Materials/M_Pickaxe_HornettaVine
graph {
 properties {
  bAutomaticallySetUsageInEditor=false
  bUseMaterialAttributes=true
  bUsedWithClothing=true
  bUsedWithMeshDeformer=true
  bUsedWithMorphTargets=true
  bUsedWithNanite=true
  bUsedWithSkeletalMesh=true
  pixelDepthOffsetMode=PDOM_Legacy
  refractionMethod=RM_IndexOfRefraction
  shadingModel=MSM_FromMaterialExpression
  subsurfaceProfile=&SS_HeroSkin_02
  usedShadingModels=Default Lit | Subsurface | Clear Coat | Subsurface Profile | Cloth | Eye
 }
 nodes [
  n1 = MaterialFunctionCall() {materialFunction=&BaseFN_CharacterFunction_PreFX}
  n2 = MaterialFunctionCall(
   PreskinnedPosition=@n1.Pre skinned Local Position
   Gmap=@n1.Pre skinned Local Position
   MaterialAttributes=@n1.Pre skinned Local Position
   AO=@n1.Pre skinned Local Position
  ) {materialFunction=&CharacterSkinFX}
  n3 = MaterialFunctionCall(AO=@n1.AO, Distance Blend=@n1.AO, Fresnel=@n1.AO, Gmap=@n1.AO, Pre Skinned Local Position=@n1.AO, Material=@n2.Result) {materialFunction=&BaseFN_CharacterFunction_PostFX}
  n4 = GetMaterialAttributes(None=@n3.Result) {attributeGetTypes [#1]}
  n5 = MaterialFunctionCall() {materialFunction=/Game/Characters/Player/Female/Medium/Bodies/F_MED_HornettaVine/Materials/MF_HornettaVine_Veins}
  n6 = ComponentMask(None=@n2.SkinFX_Mask) {g=true}
  n7 = Multiply(
   A=Multiply(A=Multiply(A=@n4.Emissive Color, B=@n5.Deep Veins), B=Add(B=@n5.Glow Gradient With Noise) {constA=1})
   B=Multiply(A=Power(Base=@n6) {constExponent=4}) {constB=4}
  )
  n8 = MaterialFunctionCall(
   Value_Morning=ScalarParameter() {defaultValue=1, group=Emissive, parameterName=EmissiveTOD_Morning}
   Value_Day=ScalarParameter() {defaultValue=1, group=Emissive, parameterName=EmissiveTOD_Day}
   Value_Evening=ScalarParameter() {defaultValue=1, group=Emissive, parameterName=EmissiveTOD_Evening}
   Value_Night=ScalarParameter() {defaultValue=1, group=Emissive, parameterName=EmissiveTOD_Night}
  ) {materialFunction=&Scalar_Time_of_Day}
  MaterialFunctionCall(In=SetMaterialAttributes(MaterialAttributes=@n4.MaterialAttributes, Emissive Color=Multiply(A=@n7, B=@n8.Result)) {attributeSetTypes [#1]}) {materialFunction=&CINE_MF_CharPostFX}
 ]
}
